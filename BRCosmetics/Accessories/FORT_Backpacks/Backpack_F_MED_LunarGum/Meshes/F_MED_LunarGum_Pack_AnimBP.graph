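# Animation Blueprint for the LunarGum backpack that adjusts the pack to spine bend and adds secondary motion to its straps and zipper bone.
# On update, it reads the owner mesh's spine_05 yaw delta from reference pose relative to spine_01, mapping it by sign to root roll and height.
# The AnimGraph applies a fixed root offset and that adjustment additively, then chains spring-driven, constrained Anim Dynamics on two strap bones and dyn_zyp.
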AnimBlueprint /BRCosmetics/Accessories/FORT_Backpacks/Backpack_F_MED_LunarGum/Meshes/F_MED_LunarGum_Pack_AnimBP
target_skeleton=&/F_MED_LunarGum_Pack_Skeleton
graphs [
 AnimGraph {nodes [
   n1 = "Transform (Modify) Bone - Bone: root"(ComponentPose=Local To Component(LocalPose=Input Pose()), Translation="0,2.5,2", Rotation="0,0,0") {
    rotationMode=BMM_Additive
    translationMode=BMM_Additive
    alphaBoolBlend {blendOption=Linear}
    boneToModify {boneName=root}
    translation {y=2.5, z=2}
   }
   n2 = "Transform (Modify) Bone - Bone: root"(ComponentPose=@n1, Translation=Get Translate_01() -> (Translate_01:Vector), Rotation=Get Rotate_01() -> (Rotate_01:Rotator)) {rotationMode=BMM_Additive, translationMode=BMM_Additive, alphaBoolBlend {blendOption=Linear}, boneToModify {boneName=root}}
   n3 = Anim Dynamics - Bone: Dyn_strap_Base(ComponentPose=@n2, Alpha=1, ExternalForce="0,0,0") {
    angularDampingOverride=1
    angularSpringConstant=250
    bAngularSpring=true
    bLinearSpring=true
    bOverrideAngularDamping=true
    bOverrideLinearDamping=true
    bUseGravityOverride=true
    linearDampingOverride=1
    linearSpringConstant=250
    simulationSpace=RootRelative
    alphaBoolBlend {blendOption=Linear}
    boundBone {boneName=Dyn_strap_Base}
    componentAppliedLinearAccClamp=<10000 10000 10000>
    componentLinearAccScale=<0.2 0.2 0.2>
    componentLinearVelScale=<0.2 0.2 0.2>
    gravityOverride {z=-980}
    physicsBodyDefinitions [{
      boundBone {boneName=Dyn_strap_Base}
      boxExtents=<8 8 8>
      constraintSetup {angularLimitsMax=<8 1 8>, angularLimitsMin=<-8 -1 -8>, angularTarget=<0 0 0>, linearAxesMax=<0.2 0.2 0.2>, linearAxesMin=<-0.2 -0.2 -0.2>}
      localJointOffset=<1 0 0>
     }]
   }
   n4 = Anim Dynamics - Bone: Dyn_strap(ComponentPose=@n3, Alpha=1, ExternalForce="0,0,0") {
    angularDampingOverride=1
    angularSpringConstant=250
    bAngularSpring=true
    bLinearSpring=true
    bOverrideAngularDamping=true
    bOverrideLinearDamping=true
    bUseGravityOverride=true
    linearDampingOverride=1
    linearSpringConstant=250
    simulationSpace=RootRelative
    alphaBoolBlend {blendOption=Linear}
    boundBone {boneName=Dyn_strap}
    componentAppliedLinearAccClamp=<10000 10000 10000>
    componentLinearAccScale=<0.1 0.1 0.1>
    componentLinearVelScale=<0.1 0.1 0.1>
    gravityOverride {z=-980}
    physicsBodyDefinitions [{
      boundBone {boneName=Dyn_strap}
      boxExtents=<8 8 8>
      constraintSetup {angularLimitsMax=<0 1 0>, angularLimitsMin=<0 -1 0>, angularTarget=<1 0 0>, linearAxesMax=<0.2 0.2 0.35>, linearAxesMin=<-0.2 -0.2 -0.35>}
      localJointOffset=<1 0 0>
     }]
   }
   n5 = Anim Dynamics - Bone: dyn_zyp(ComponentPose=@n4, Alpha=1, ExternalForce="0,0,0") {
    angularDampingOverride=1
    angularSpringConstant=250
    bAngularSpring=true
    bOverrideAngularDamping=true
    bUseGravityOverride=true
    linearDampingOverride=1
    linearSpringConstant=250
    simulationSpace=RootRelative
    alphaBoolBlend {blendOption=Linear}
    boundBone {boneName=dyn_zyp}
    componentAppliedLinearAccClamp=<5000 5000 5000>
    componentLinearAccScale=<0.75 0.75 0.75>
    componentLinearVelScale=<0.75 0.75 0.75>
    gravityOverride {z=-980}
    physicsBodyDefinitions [{
      boundBone {boneName=dyn_zyp}
      boxExtents=<10 10 10>
      constraintSetup {angularLimitsMax=<1 110 1>, angularLimitsMin=<-1 0 -1>, angularTarget=<1 0 0>}
      localJointOffset=<-1 0 0>
     }]
   }
   Output Pose(Result=Component To Local(ComponentPose=@n5)) {layerGroup=None}
  ]}
 EventGraph {nodes [
   Event BlueprintUpdateAnimation()
   > Sequence()
   n1 = GetOwnerSkeletalMeshComponent()
   n2 = BreakRotator(InRot=BreakTransform(InTransform=GetDeltaTransformFromRefPose(self=@n1, BoneName=spine_05, BaseName=spine_01)).Rotation)
   > n3 = Branch(Condition="float > float"(A=@n2.Yaw))
   > Set Rotate_01(Rotate_01:Rotator=MakeRotator(Roll=MapRangeClamped(Value=@n2.Yaw, InRangeA=0, InRangeB=40, OutRangeB=0))) -> (Output_Get:Rotator)
   > Set Translate_01(Translate_01:Vector=MakeVector(Y=0, Z=MapRangeClamped(Value=@n2.Yaw, InRangeA=0, InRangeB=40, OutRangeB=0))) -> (Output_Get:Vector)
   Set Rotate_01(execute=[@n3.else, @n4.OutputPin], Rotate_01=MakeRotator(Roll=MapRangeClamped(Value=@n2.Yaw, InRangeA=-15, InRangeB=-45, OutRangeB=25)))
   > Set Translate_01(Translate_01=MakeVector(Y=0, Z=MapRangeClamped(Value=@n2.Yaw, InRangeA=-15, InRangeB=-45, OutRangeB=2)))
   n4 = Reroute Node()
   TryGetPawnOwner()
  ]}
]
variables [
 Rotate_01 {type=Rotator}
 Translate_01 {type=Vector}
]
